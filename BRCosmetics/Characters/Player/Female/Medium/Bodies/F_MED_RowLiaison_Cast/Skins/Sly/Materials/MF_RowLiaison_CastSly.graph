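# Material function for the RowLiaison Cast Sly skin that layers an animated, reactive glowing pattern onto the character's incoming material attributes.
# Time since a Blueprint-supplied timestamp samples a curve-atlas row for reactive progress.
# RGB pattern masks drive phase-offset sine waves through noise, with separate idle and reactive flows, panned pre-skinned-position highlights and Fresnel shaping.
# The result adds emissive and tints base color, while pattern masks also adjust metallic and roughness.
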MaterialFunction /BRCosmetics/Characters/Player/Female/Medium/Bodies/F_MED_RowLiaison_Cast/Skins/Sly/Materials/MF_RowLiaison_CastSly
graph {nodes [
  n1 = GetMaterialAttributes(None=FunctionInput() {inputName=Material, inputType=FunctionInput_MaterialAttributes}) {attributeGetTypes [
    #1
    #2
    #3
    #4
   ]}
  n2 = VectorParameter() {group=RowLiaison_CastSly, parameterName=RLCS_PatternBasecolor, defaultValue=<rgba 0.3698 0.3378 0.2639 1>}
  n3 = NamedRerouteUsage()
  n4 = VectorParameter() {group=RowLiaison_CastSly, parameterName=RLCS_ReactiveBasecolor, defaultValue=<rgba 1.5 1.2997 0.8359 1>}
  n5 = SetMaterialAttributes!(
   MaterialAttributes=@n1.MaterialAttributes
   EmissiveColor=Add(A=@n1.EmissiveColor, B=NamedRerouteUsage())
   BaseColor=LinearInterpolate(A=LinearInterpolate(A=@n1.BaseColor, B=@n2.RGB, Alpha=@n3), B=@n4.RGB, Alpha=NamedRerouteUsage())
   Metallic=LinearInterpolate(A=@n1.Metallic, B=ScalarParameter() {defaultValue=1, parameterName=RLCS_PatternMetallic}, Alpha=@n3)
   Specular
   Roughness=LinearInterpolate(A=@n1.Roughness, Alpha=@n3) {constB=0.25}
  ) {attributeSetTypes [
    #1
    #2
    #3
    #5
    #4
   ]}
  FunctionOutput(None=@n5)
  n6 = Subtract(A=Time(), B=ScalarParameter() {defaultValue=10000000272564224, group=RowLiaison_CastSly, parameterName=RLCS_GameTimeFromBP})
  n7 = ScalarParameter() {defaultValue=5, group=RowLiaison_CastSly, parameterName=RLCS_ReactiveTime}
  n8 = CurveAtlasRowParameter(CurveTime=Saturate(None=Divide(A=@n6, B=@n7))) {
   atlas=/BRCosmetics/Characters/Player/Female/Medium/Bodies/F_MED_RowLiaison_Cast/Skins/Sly/Textures/CA_RowLiaison_CastSly_Reactivity
   curve=/BRCosmetics/Characters/Player/Female/Medium/Bodies/F_MED_RowLiaison_Cast/Skins/Sly/Textures/C_RowLiaison_CastSly_Reactivity
   group=RowLiaison_CastSly
   parameterName=RLCS_ReactiveCurve
  }
  LinearInterpolate(
   A=@n8.R
   B=Saturate(None=ScalarParameter() {group=RowLiaison_CastSly, parameterName=RLCS_DebugProgress})
   Alpha=ScalarParameter() {group=RowLiaison_CastSly, parameterName=RLCS_UseDebug}
  )
  n9 = VectorParameter() {group=RowLiaison_CastSly, parameterName=RLCS_ReactiveEmissiveColor, defaultValue=<rgba 1 0.8665 0.5573 1>}
  n10 = Multiply(A=Time(), B=ScalarParameter() {defaultValue=-0.15, group=RowLiaison_CastSly, parameterName=RLCS_PatternFlowSpeed_ReactiveOff}) {constB=-0.5}
  n11 = Add(A=NamedRerouteUsage(), B=@n10)
  n12 = Multiply(A=NamedRerouteUsage(), B=SphereMask(A=Sine(None=@n11), B=Constant() {r=1}) {attenuationRadius=2, hardnessPercent=0})
  n13 = Add(A=NamedRerouteUsage(), B=Add(A=@n10) {constB=0.3245})
  n14 = Multiply(A=NamedRerouteUsage(), B=SphereMask(A=Sine(None=@n13), B=Constant() {r=1}) {attenuationRadius=2, hardnessPercent=0})
  n15 = NamedRerouteUsage()
  n16 = NamedRerouteUsage()
  n17 = Add(A=@n10) {constB=0.6844}
  n18 = SphereMask(A=Sine(None=Add(A=@n16, B=@n17)), B=Constant() {r=1}) {attenuationRadius=2, hardnessPercent=0}
  n19 = Multiply(A=Time(), B=ScalarParameter() {defaultValue=-0.25, group=RowLiaison_CastSly, parameterName=RLCS_PatternFlowSpeed_ReactiveOn}) {constB=-0.5}
  n20 = Sine(None=Add(A=NamedRerouteUsage(), B=@n19))
  n21 = Multiply(A=NamedRerouteUsage(), B=Power(Base=ConstantBiasScale(None=@n20)))
  n22 = Sine(None=Add(A=NamedRerouteUsage(), B=Add(A=@n19) {constB=0.3245}))
  n23 = Multiply(A=NamedRerouteUsage(), B=Power(Base=ConstantBiasScale(None=@n22)))
  n24 = NamedRerouteUsage()
  n25 = Add(A=NamedRerouteUsage(), B=Add(A=@n19) {constB=0.6844})
  n26 = Power(Base=ConstantBiasScale(None=Sine(None=@n25)))
  n27 = LinearInterpolate(
   A=Add(A=Add(A=@n12, B=@n14), B=Multiply(A=@n15, B=@n18))
   B=Add(A=Add(A=@n21, B=@n23), B=Multiply(A=@n24, B=@n26))
   Alpha=NamedRerouteUsage()
  )
  n28 = NamedRerouteUsage()
  n29 = Constant() {r=0.12}
  n30 = Constant() {r=1}
  n31 = NamedRerouteUsage()
  n32 = Fresnel() {baseReflectFraction=0, exponent=2}
  n33 = Multiply(
   A=Multiply(
    A=MaterialFunctionCall!(Base=@n27, Add=@n28) {materialFunction=/Engine/Functions/Engine_MaterialFunctions02/Math/MultiplyAdd}.Result
    B=LinearInterpolate(A=@n29, B=@n30, Alpha=@n31)
   )
   B=OneMinus(None=@n32)
  )
  n34 = Multiply(A=Multiply(A=@n32, B=NamedRerouteUsage()), B=Constant() {r=0.5}) {constB=0.75}
  n35 = Multiply(
   A=LinearInterpolate(A=@n33, B=Add(A=@n33, B=@n34), Alpha=NamedRerouteUsage())
   B=ScalarParameter() {defaultValue=1, group=RowLiaison_CastSly, parameterName=RLCS_Reactivity}
  )
  n36 = Multiply(A=@n9.RGB, B=NamedRerouteDeclaration(None=Saturate(None=@n35)) {name=FinalPattern, nodeColor=<rgba 0.8984 0 1 1>})
  NamedRerouteDeclaration(None=@n36) {name=Emissive_OUT, nodeColor=<rgba 0 0.0156 1 1>}
  MaterialExpressionTextureSample_4 = TextureSample() {
   samplerType=SAMPLERTYPE_Masks
   texture=/BRCosmetics/Characters/Player/Female/Medium/Bodies/F_MED_RowLiaison_Cast/Skins/Sly/Textures/T_RowLiaison_CastSly_Noise
  }
  n37 = Convert(None=@MaterialExpressionTextureSample_4.RGB) {
   nodeName=Break Float3
   convertInputs [{
     type=Vector3
     expressionInput {
      expression=&self:MaterialExpressionTextureSample_4
      inputName=None
      mask=1
      maskA=0
      maskB=1
      maskG=1
      maskR=1
      outputIndex=0
     }
    }]
   convertOutputs [{type=Scalar}, {type=Scalar}, {type=Scalar}]
  }
  NamedRerouteDeclaration(None=@n37) {name=RLCS_PatternFlow_R, nodeColor=<rgba 1 0.1649 0.0879 1>}
  NamedRerouteDeclaration(None=@n37) {name=RLCS_PatternFlow_B, nodeColor=<rgba 0 0.0819 1 1>}
  n38 = ComponentMask(None=NamedRerouteUsage()) {b=true, r=true}
  n39 = TextureSample(UVs=Panner(Coordinate=@n38, Speed=Constant2Vector() {g=-0.0864, r=0.1353}) {bFractionalPart=true, speedX=0.197, speedY=-0.1235}) {
   samplerSource=SSM_Wrap_WorldGroupSettings
   samplerType=SAMPLERTYPE_Masks
   texture=&T_RowLiaison_CastSly_Noise
  }
  n40 = TextureSample(UVs=Panner(Coordinate=@n38, Speed=Constant2Vector() {g=-0.1795, r=-0.0374}) {bFractionalPart=true, speedX=-0.324, speedY=-0.0345}) {
   samplerSource=SSM_Wrap_WorldGroupSettings
   samplerType=SAMPLERTYPE_Masks
   texture=&T_RowLiaison_CastSly_Noise
  }
  n41 = Multiply(A=@n39.A, B=@n40.A)
  NamedRerouteDeclaration(None=Multiply(A=Power(Base=@n41) {constExponent=3}, B=Constant() {r=15})) {name=HighLight, nodeColor=<rgba 0 0.5078 1 1>}
  NamedRerouteDeclaration(None=@n36) {name=Basecolor_OUT, nodeColor=<rgba 0 1 0.2266 1>}
  NamedRerouteDeclaration(None=@n8.R) {name=ReactiveProgress, nodeColor=<rgba 1 0 0.2578 1>}
  n42 = VertexInterpolator(VS=LocalPosition() {localOrigin=InstancePreSkinning}.XYZ)
  NamedRerouteDeclaration(None=Divide(A=FunctionInput(Preview=@n42.PS) {inputName=PreskinnedLocPos, sortPriority=1}) {constB=64}) {name=PreskinnedLocPos, nodeColor=<rgba 1 0.0703 0 1>}
  n43 = TextureSampleParameter2D() {
   group=RowLiaison_CastSly
   parameterName=RLCS_Texture
   samplerType=SAMPLERTYPE_LinearColor
   texture=/BRCosmetics/Characters/Player/Female/Medium/Bodies/F_MED_RowLiaison_Cast/Skins/Sly/Textures/T_RowLiaison_CastSly_ReactiveMask
  }
  NamedRerouteDeclaration(None=@n43.A) {name=RLCS_DressMask, nodeColor=<rgba 1 0.7079 0.9608 1>}
  MaterialExpressionMultiply_12 = Multiply(A=Power(Base=@n43.RGB, Exp=Constant() {r=3}), B=Constant() {r=1.75})
  n44 = Convert(None=@MaterialExpressionMultiply_12) {
   nodeName=Break Float3
   convertInputs [{
     type=Vector3
     expressionInput {
      expression=&self:MaterialExpressionMultiply_12
      inputName=None
      mask=0
      maskA=0
      maskB=0
      maskG=0
      maskR=0
      outputIndex=0
     }
    }]
   convertOutputs [{type=Scalar}, {type=Scalar}, {type=Scalar}]
  }
  n45 = Max(
   A=NamedRerouteDeclaration(None=@n44) {name=RLCS_PatternMask_R, nodeColor=<rgba 1 0.1649 0.0879 1>}
   B=NamedRerouteDeclaration(None=@n44) {name=RLCS_PatternMask_G, nodeColor=<rgba 0 1 0.2919 1>}
  )
  n46 = NamedRerouteDeclaration(None=@n44) {name=RLCS_PatternMask_B, nodeColor=<rgba 0 0.0819 1 1>}
  NamedRerouteDeclaration(None=Multiply(A=Max(A=@n45, B=@n46), B=ScalarParameter() {defaultValue=1, group=RowLiaison_CastSly, parameterName=RLCS_Reactivity})) {name=PatternMaskRGB, nodeColor=<rgba 0.875 0 1 1>}
  NamedRerouteDeclaration(None=@n37) {name=RLCS_PatternFlow_G, nodeColor=<rgba 0 1 0.2919 1>}
 ]}
