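# Parent material for the Sly skin of the RowLiaison_Cast female character body, masked and skeletal-mesh ready, with shading model set by expression.
# It chains shared Character_Parent_2023 functions through material attributes: PreFX feeds OpacityMask, then skin-specific MF_RowLiaison_CastSly, then PostFX.
# PreFX also passes pre-skinned local position to the Sly function and auxiliary data to PostFX.
# The skin's custom look lives in that inserted function, keeping this graph pure wiring.
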Material /BRCosmetics/Characters/Player/Female/Medium/Bodies/F_MED_RowLiaison_Cast/Skins/Sly/Materials/M_RowLiaison_CastSly_Parent
graph {
 properties {
  bUseMaterialAttributes=true
  bUsedWithMeshDeformer=true
  bUsedWithMorphTargets=true
  bUsedWithSkeletalMesh=true
  blendMode=BLEND_Masked
  pixelDepthOffsetMode=PDOM_Legacy
  shadingModel=MSM_FromMaterialExpression
  usedShadingModels=Default Lit | Subsurface | Clear Coat | Subsurface Profile | Cloth
 }
 nodes [
  n1 = MaterialFunctionCall() {materialFunction=&MF_CharacterParent_2023_PreFX}
  MaterialFunctionCall(
   Material=MaterialFunctionCall!(
    Material=MaterialFunctionCall!(Result=@n1.Result) {materialFunction=/Game/Global/Materials/Character_Parent_2023/Functions/MF_CharacterParent_2023_OpacityMask}."Switch Param (False)"
    PreskinnedLocPos=@n1.Pre skinned Local Position
   ) {materialFunction=&/MF_RowLiaison_CastSly}.Result
   M=@n1.M
   Pre Skinned Local Position=@n1.M
   Distance Blend=@n1.M
   Fresnel=@n1.M
  ) {materialFunction=&MF_CharacterParent_2023_PostFX}
 ]
}
